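# Sound Cue for a sharp mud explosion impact, mixing close and distant layers through separate water-explosion attenuations in the explosives sound class.
# The close layer blends a randomly pitched water bullet impact with a slightly delayed, octave-up mud burst.
# The distant layer plays a slightly delayed, randomly pitched distant mud burst.
# Each source picks from four variants, and the whole cue is pitched up with its own concurrency limit.
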SoundCue /Game/Sounds/Fort_Weapons/Impacts/Mud/Mud_Explosion_Sharp_Cue
root=Mixer_5
nodes [
 {
  id=Attenuation_0
  children [Mixer_1]
  properties {
   attenuationSettings=/Game/Sounds/Attenuation/Fort_Explosion_Water_Close_Athena_Att
   attenuationOverrides {falloffDistance=900, attenuationShapeExtents {x=100}}
  }
 }
 {id=Attenuation_4, children [Delay_3], properties {attenuationSettings=/Game/Sounds/Attenuation/Fort_Explosion_Water_Distant_Athena_Att}}
 {id=Delay_0, children [Modulator_1], properties {delayMax=0.1, delayMin=0.1}}
 {id=Delay_3, children [Modulator_5], properties {delayMax=0.1, delayMin=0.1}}
 {id=Mixer_1, children [Delay_0,Modulator_0], properties {inputVolume [0.8,1]}}
 {id=Mixer_5, children [Attenuation_0, Attenuation_4], properties {inputVolume [0.7,0.7]}}
 {id=Modulator_0, children [Random_5], properties {pitchMax=1.1, pitchMin=0.85, volumeMax=1, volumeMin=1}}
 {id=Modulator_1, children [Random_6], properties {pitchMax=2, pitchMin=2, volumeMax=1, volumeMin=1}}
 {id=Modulator_5, children [Random_12], properties {pitchMax=1.1, pitchMin=0.85, volumeMax=1, volumeMin=1}}
 {
  id=Random_12
  children [WavePlayer_43, WavePlayer_44, WavePlayer_45, WavePlayer_46]
  properties {bShouldExcludeFromBranchCulling=true, weights [1,1,1,1]}
 }
 {id=Random_5, children [WavePlayer_0,WavePlayer_1,WavePlayer_2,WavePlayer_3], properties {weights [1,1,1,1]}}
 {
  id=Random_6
  children [WavePlayer_22, WavePlayer_23, WavePlayer_24, WavePlayer_25]
  properties {bShouldExcludeFromBranchCulling=true, weights [1,1,1,1]}
 }
 {id=WavePlayer_0, properties {soundWaveAssetPtr=/Game/Sounds/Fort_Weapons/Impacts/Bullet_Impacts_S10/Bullet_Impact_Water_Med_01}}
 {id=WavePlayer_1, properties {soundWaveAssetPtr=/Game/Sounds/Fort_Weapons/Impacts/Bullet_Impacts_S10/Bullet_Impact_Water_Med_02}}
 {id=WavePlayer_2, properties {soundWaveAssetPtr=/Game/Sounds/Fort_Weapons/Impacts/Bullet_Impacts_S10/Bullet_Impact_Water_Med_03}}
 {id=WavePlayer_22, properties {soundWaveAssetPtr=&/Mud_Explo_Medium_Burst_Close_01}}
 {id=WavePlayer_23, properties {soundWaveAssetPtr=&/Mud_Explo_Medium_Burst_Close_02}}
 {id=WavePlayer_24, properties {soundWaveAssetPtr=&/Mud_Explo_Medium_Burst_Close_03}}
 {id=WavePlayer_25, properties {soundWaveAssetPtr=&/Mud_Explo_Medium_Burst_Close_04}}
 {id=WavePlayer_3, properties {soundWaveAssetPtr=/Game/Sounds/Fort_Weapons/Impacts/Bullet_Impacts_S10/Bullet_Impact_Water_Med_04}}
 {id=WavePlayer_43, properties {soundWaveAssetPtr=&/Mud_Explo_Medium_Burst_Distant_01}}
 {id=WavePlayer_44, properties {soundWaveAssetPtr=&/Mud_Explo_Medium_Burst_Distant_02}}
 {id=WavePlayer_45, properties {soundWaveAssetPtr=&/Mud_Explo_Medium_Burst_Distant_03}}
 {id=WavePlayer_46, properties {soundWaveAssetPtr=&/Mud_Explo_Medium_Burst_Distant_04}}
]
properties {
 bOverrideConcurrency=true
 maxDistance=10000
 pitchMultiplier=1.5
 soundClassObject=/Game/Sounds/FortSoundClassesAndModes/Fort_Explosives
 volumeMultiplier=0.708
 concurrencyOverrides {maxCount=6}
}
